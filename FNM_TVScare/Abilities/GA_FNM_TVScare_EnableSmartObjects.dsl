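; Gameplay ability that enables the TV scare smart objects in the level.
; On activation it waits briefly, finds every SO_TVScare actor with GetAllActorsOfClass, and calls SetSmartObjectEnabled on each one's SmartObject component under the Gameplay reason.
; Any failed enable writes a shipping log message, and the ability ends after the loop.
; The search has no distance filter, although the log message refers to nearby smart objects.
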(event Ability|EventActivateAbility
  (Utilities|FlowControl|Delay 2.0)
  (bind _outactors (Actor|GetAllActorsOfClass "/FNM_TVScare/SmartObjects/SO_TVScare.SO_TVScare_C"))
  (for _array_element _outactors
    (if (not (SmartObject|SetSmartObjectEnabled(defaultreason:Gameplay) (Variables|/SmartObjectLibrary/SmartObject_NoMesh_Actor.SmartObject_NoMesh_Actor_C|GetSmartObject _array_element) true :!class /Script/SmartObjectsModule.SmartObjectComponent))
      (Fortnite|FortShippingLog self "Failed to Activate Nearby Smart Objects" true)))
  (Ability|EndAbility))

(event Ability|EventOnEndAbility
  (params (bWasCancelled :type Boolean)))

(event Ability|EventFailedToActivatePassiveAbility
  (params (ActorInfo :type Gameplay%20Ability%20Actor%20Info%20Structure)))
