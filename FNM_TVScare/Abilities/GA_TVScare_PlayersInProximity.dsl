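; Gameplay ability for the TV Scare feature, named for affecting players in proximity.
; On activation it runs a DoTargetSelection task under the first targeting application tag, and each Targeted callback applies the gameplay effect container using the returned target data and tag.
; The targeting rules and effects are not defined in this graph, which also never calls EndAbility and leaves its end event empty.
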(event Ability|EventActivateAbility
  (bind (_asynctaskproxy _targetdata _applicationtag) (Ability|Tasks|DoTargetSelection "None" "(TagName=\"Targeting.ApplicationTag.0\")")
    (:Targeted
      (Ability|ApplyGameplayEffectContainer _targetdata _applicationtag))))

(event Ability|EventOnEndAbility
  (params (bWasCancelled :type Boolean)))
